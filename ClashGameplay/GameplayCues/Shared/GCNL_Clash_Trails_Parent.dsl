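; Parent looping gameplay cue for Clash trail effects, handling team color and audio.
; Loop start sets a color bool parameter from FXIsRed on the first particle component.
; AudioComponentSetup picks the loop or stop sound and sets IsLocal and IsTeammate parameters from view target and team affiliation.
; Removal and cease-relevant call it on the stop path; CheckTrails and application events are empty here.
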(event EventOnLoopingStart(EitherVFX-preferthisversion)
  (params (MyTarget :type Actor%20Object%20Reference) (Parameters :type Gameplay%20Cue%20Parameters%20Structure%20%28by%20ref%29) (ParticleComponents :type Array%20of%20FXSystem%20Component%20Object%20References) (AudioComponents :type Array%20of%20Audio%20Component%20Object%20References))
  ;[1 1 1] Save Out refs to the FX, set up the timer and call the color set, for the trails, spawn it dynamically so it can be toggled.
  (Utilities|FlowControl|Sequence
    (:then_0
      (bind _result (ArrayMacros|SafeArrayGet ParticleComponents)
        (:%20%20
          (Effects|Components|ParticleSystem|SetBoolParameter _result (Variables|Default|GetFXColorParam) (Variables|Default|GetFXIsRed)))))))

(event Custom|CheckTrails)

(event EventOnRemoval(EitherVFX-preferthisversion)
  (params (MyTarget :type Actor%20Object%20Reference) (Parameters :type Gameplay%20Cue%20Parameters%20Structure%20%28by%20ref%29) (ParticleComponents :type Array%20of%20FXSystem%20Component%20Object%20References) (AudioComponents :type Array%20of%20Audio%20Component%20Object%20References) (BurstCameraShakeInstance :type Camera%20Shake%20Base%20Object%20Reference) (BurstDecalInstance :type Decal%20Actor%20Object%20Reference))
  ;[1 1 1] Removal
  (bind _result (ArrayMacros|SafeArrayGet AudioComponents)
    (:%20%20
      (CallFunction|AudioComponentSetup _result))))

(event EventOnApplication(EitherVFX-preferthisone)
  (params (MyTarget :type Actor%20Object%20Reference) (Parameters :type Gameplay%20Cue%20Parameters%20Structure%20%28by%20ref%29) (ParticleComponents :type Array%20of%20FXSystem%20Component%20Object%20References) (AudioComponents :type Array%20of%20Audio%20Component%20Object%20References) (BurstCameraShakeInstance :type Camera%20Shake%20Base%20Object%20Reference) (BurstDecalInstance :type Decal%20Actor%20Object%20Reference)))

(fn AudioComponentSetup
  (params (AudioComponent :type Audio%20Component%20Object%20Reference) (Starting :type Boolean))
  (bind _output_get (LocalVariables|SetLocalAudioComponent AudioComponent))
  (Audio|Components|Audio|SetSound _output_get (select Starting (Variables|Audio|GetSoundOnLoop) (Variables|Audio|GetSoundOnStop)))
  (bind _output_get_2 (LocalVariables|SetLocalFPC (Fortnite|GetFirstLocalFortPlayerController (Variables|Default|GetAthenaPlayerPawn) true true)))
  (Utilities|IsValid _output_get_2
    (:Is%20Valid
      (Audio|Parameter|SetIntegerParameter (LocalVariables|GetLocalAudioComponent) "IsLocal" (Math|Conversions|ToInteger(Boolean) (Utilities|Equal(Object) (Pawn|GetViewTarget _output_get_2) (Variables|Default|GetAthenaPlayerPawn))) :!class /Script/AudioExtensions.AudioParameterControllerInterface)
      (Audio|Parameter|SetIntegerParameter (LocalVariables|GetLocalAudioComponent) "IsTeammate" (Math|Conversions|ToInteger(Boolean) (Utilities|Enum|Equal(Enum) (Team|GetTeamAffiliationtoLocalPlayer 0 (Team|GetActorTeam (Variables|Default|GetAthenaPlayerPawn)) true true) "Friendly")) :!class /Script/AudioExtensions.AudioParameterControllerInterface))))

(fn OnCeaseRelevant
  (params (MyTarget :type Actor%20Object%20Reference) (Parameters :type Gameplay%20Cue%20Parameters%20Structure%20%28by%20ref%29))
  (bind _result (ArrayMacros|SafeArrayGet)
    (:%20%20
      (CallFunction|AudioComponentSetup _result))))

(fn ConstructionScript)
